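# Animated purple material for the StickyPurple cosmetic vehicle wheel, built on material attributes with lit shading.
# Rotating radial and noise textures, warped and bump-offset for fake depth, combine with a center glow and panning, V-masked edge clouds on UV1.
# These masks blend parameterized colors, add center and Fresnel tints, and are multiplied by a noise mask.
# The result drives Base Color and intensity-scaled Emissive, alongside scalar Metallic and Roughness.
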Material /VehicleCosmetics/Wheels/StickyPurple/Materials/M_Wheel_StickyPurple_FX
graph {
 properties {
  bAutomaticallySetUsageInEditor=false
  bUseMaterialAttributes=true
  bUsedWithMeshDeformer=true
  bUsedWithSkeletalMesh=true
  previewMesh=/VehicleCosmetics/Wheels/StickyPurple/Meshes/wheel_StickyPurple_SM
  usedShadingModels=Default Lit
 }
 nodes [
  n1 = TextureObjectParameter() {group=00 Base, parameterName=Noise, sortPriority=0, texture=/VehicleCosmetics/Wheels/StickyPurple/Textures/wheel_StickyPurple_Noise1}
  NamedRerouteDeclaration(None=@n1) {name=Noise, nodeColor=<rgba 1 0 0.9141 1>}
  n2 = VectorParameter() {parameterName=Color 2, defaultValue=<rgba 0.227 0 0.3864 1>}
  n3 = VectorParameter() {parameterName=Color 1, defaultValue=<rgba 0.6724 0 0.6038 1>}
  n4 = TextureSample(UVs=Rotator() {speed=-0.2}) {samplerSource=SSM_Wrap_WorldGroupSettings, texture=/VehicleCosmetics/Wheels/_SharedTextures/FX/wheel_ringer_a_radial01}
  n5 = Add(A=@n4.R, B=@n4.R)
  n6 = BumpOffset(Height=Constant() {r=-4})
  n7 = NamedRerouteUsage()
  n8 = Add(
   A=Multiply(A=@n4.R) {constB=0.4}
   B=Multiply(A=TextureSample(UVs=@n6, Tex=@n7) {samplerSource=SSM_Wrap_WorldGroupSettings, texture=&wheel_StickyPurple_Noise1}.G) {constB=0.3}
  )
  n9 = Rotator() {speed=0.1}
  n10 = TextureSample(UVs=BumpOffset(Coordinate=Add(A=@n8, B=@n9), Height=Constant() {r=-4}), Tex=NamedRerouteUsage()) {samplerSource=SSM_Wrap_WorldGroupSettings, texture=&wheel_StickyPurple_Noise1}
  n11 = Add(A=@n10.R, B=@n10.R)
  n12 = TextureSampleParameter2D(UVs=BumpOffset(Coordinate=TextureCoordinate(), Height=ScalarParameter() {defaultValue=-3, group=00 Base, parameterName=Center Glow Bump Offset})) {group=00 Base, parameterName=Center, sortPriority=1, texture=/VehicleCosmetics/Wheels/StickyPurple/Textures/wheel_StickyPurple_Noise2}
  n13 = TextureCoordinate() {coordinateIndex=1, uTiling=2}
  n14 = Multiply(A=Time(), B=ScalarParameter() {defaultValue=0.8, group=00 Base, parameterName=Edge Clouds Speed})
  n15 = Panner(Coordinate=Panner(Coordinate=@n13, Time=Constant() {r=0.5}) {speedY=1}, Time=@n14) {speedX=0.1}
  n16 = TextureCoordinate() {coordinateIndex=1}
  n17 = Multiply(A=TextureSample(UVs=Panner(Coordinate=@n16, Time=@n14) {speedX=0.2}, Tex=NamedRerouteUsage()) {samplerSource=SSM_Wrap_WorldGroupSettings, texture=&wheel_StickyPurple_Noise1}.G) {constB=0.2}
  n18 = ComponentMask(None=TextureCoordinate() {coordinateIndex=1}) {g=true}
  n19 = Multiply(
   A=TextureSample(UVs=Add(A=@n15, B=@n17), Tex=NamedRerouteUsage()) {samplerSource=SSM_Wrap_WorldGroupSettings, texture=&wheel_StickyPurple_Noise1}.R
   B=Clamp(None=ConstantBiasScale(None=@n18) {bias=-0.4, scale=10})
  )
  n20 = Add(A=Add(A=Multiply(A=@n5, B=@n11), B=@n12.R), B=Clamp(None=ConstantBiasScale(None=@n19) {bias=-0.12, scale=2}))
  n21 = Clamp(None=@n20)
  n22 = ConstantBiasScale(None=@n20) {bias=0, scale=3}
  n23 = Panner(Coordinate=TextureCoordinate() {coordinateIndex=1, uTiling=0.5}, Time=Constant() {r=0.5}) {speedX=0.5, speedY=1}
  n24 = TextureSample(UVs=Panner(Coordinate=TextureCoordinate() {coordinateIndex=1, uTiling=0.5}, Time=@n14) {speedX=0.2}, Tex=NamedRerouteUsage()) {samplerSource=SSM_Wrap_WorldGroupSettings, texture=&wheel_StickyPurple_Noise1}
  n25 = TextureSample(UVs=Add(A=Panner(Coordinate=@n23, Time=@n14) {speedX=0.1}, B=Multiply(A=@n24.G) {constB=0.2}), Tex=NamedRerouteUsage()) {samplerSource=SSM_Wrap_WorldGroupSettings, texture=&wheel_StickyPurple_Noise1}
  n26 = TextureCoordinate() {coordinateIndex=1}
  n27 = Clamp(None=ConstantBiasScale(None=ComponentMask(None=@n26) {g=true}) {bias=-0.4, scale=10})
  n28 = Clamp(None=ConstantBiasScale(None=Multiply(A=@n25.R, B=@n27)) {bias=-0.12, scale=2})
  n29 = LinearInterpolate(
   A=VectorParameter() {parameterName=Color 3, defaultValue=<rgba 0.107 0 0.1779 1>}.RGB
   B=Add(A=LinearInterpolate(A=@n2.RGB, B=@n3.RGB, Alpha=@n21), B=Multiply(A=@n21) {constB=0.1})
   Alpha=Multiply(A=Clamp(None=@n22), B=OneMinus(None=@n28))
  )
  n30 = Multiply(A=VectorParameter() {parameterName=Color Center, defaultValue=<rgba 0.9047 0.7835 1 1>}.RGB, B=@n12.R)
  n31 = VectorParameter() {parameterName=Color Fresnel, defaultValue=<rgba 0.107 0 0.1779 1>}
  n32 = MaterialFunctionCall!(Input Normal, Input Sharpness=Constant() {r=2}, Input Two Sided) {materialFunction=/VehicleCosmetics/Wheels/CarCar_Holographic/Materials/MF_Fresnel_Advanced}
  n33 = TextureSample(UVs=@n13, Tex=NamedRerouteUsage()) {samplerSource=SSM_Wrap_WorldGroupSettings, texture=&wheel_StickyPurple_Noise1}
  n34 = Multiply(A=Add(A=Add(A=@n29, B=@n30), B=Multiply(A=@n31.RGB, B=@n32.Edge)), B=Clamp(None=ConstantBiasScale(None=@n33.B) {bias=0.01, scale=10}))
  NamedRerouteDeclaration(None=Multiply(A=@n34, B=ScalarParameter() {group=00 Base, parameterName=Emission Intensity, sortPriority=2})) {name=Emissive, nodeColor=<rgba 1 0.4453 0 1>}
  NamedRerouteDeclaration(None=@n34) {name=BC, nodeColor=<rgba 0.125 0 1 1>}
  SetMaterialAttributes!(
   MaterialAttributes
   Base Color=NamedRerouteUsage()
   Emissive Color=NamedRerouteUsage()
   Metallic=ScalarParameter() {group=00 Base, parameterName=Metallic, sortPriority=2}
   Roughness=ScalarParameter() {defaultValue=0.2, group=00 Base, parameterName=Roughness, sortPriority=3}
  ) {attributeSetTypes [
    #1
    #2
    #3
    #4
   ]}
 ]
}
